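# Unlit debug material in the volumetric fluid simulation tools for viewing simulation boundaries.
# It calls the FluidBoundary_3D material function and uses only its Positive X output.
# MakeFloat3 copies that value into all three channels to form a grayscale vector, though the graph shows no link from it to a material output.
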Material /Volumetrics/Tools/FluidSimulation/Materials/Debug/Boundaries
graph {
 properties {pixelDepthOffsetMode=PDOM_Legacy, shadingModel=MSM_Unlit, usedShadingModels=Unlit}
 nodes [
  n1 = MaterialFunctionCall() {materialFunction=/Volumetrics/Tools/FluidSimulation/MaterialFunctions/FluidBoundary_3D}
  MaterialFunctionCall(X=@n1.Positive X, Y=@n1.Positive X, Z=@n1.Positive X) {materialFunction=&MakeFloat3}
 ]
}
