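# Material function that maps 2D UVs to 3D UVW coordinates for rendering tiling pseudo volume textures into a flipbook-style render target.
# UVs split into a frame grid, with each cell's fraction rescaled and offset by the edge pad in texels, extending past 0-1 so borders repeat.
# W is the cell's linear index divided by X frames squared, assuming a square grid.
# The commented precision offset on W currently adds zero.
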MaterialFunction /Volumetrics/Content/VolumeTextures/MaterialFunctions/EncodeVolumeCoordinates_Tiling
graph {
 properties {bExposeToLibrary=true, description=Provides 3D UVW coordinates laid out in a 2D array for rendering to Psuedo Volume Textures.}
 nodes [
  n1 = FunctionInput() {
   bUsePreviewValueAsDefault=true
   description="The number of X, Y frames to use per side."
   inputName=XY Frames
   inputType=FunctionInput_Vector2
   previewValue=<8 1 0 1>
  }
  n2 = Multiply(A=TextureCoordinate(), B=ComponentMask(None=@n1) {g=true, r=true})
  n3 = FunctionInput() {
   bUsePreviewValueAsDefault=true
   description=The resolution of the render target the volume texture will be rendered to.
   inputName=Target Resolution
   inputType=FunctionInput_Vector2
   previewValue=<1024 1024 0 1>
  }
  n4 = Divide(A=@n3, B=@n1) {constA=2048, constB=8, desc="2048 / 8 = 256  pixels per cell"}
  n5 = FunctionInput() {
   bUsePreviewValueAsDefault=true
   description=How many texels per side to repeat. A value of 1 adds 2 extra texels and has 1 seamless mip. A value of 2 adds 4 extra texels and has 2 seamless mips. A value of 4 adds 8 extra texels and has 3 seamless mips and so on.
   inputName=Edge Pad Texels
   inputType=FunctionInput_Scalar
   previewValue=<4 0 0 1>
  }
  n6 = Multiply(A=Frac(None=@n2), B=Divide(A=Add(A=@n4, B=@n5) {constB=4, desc="Pad 4 pixels, 2 per side"}, B=@n4))
  n7 = Multiply(A=Divide(B=@n4) {desc=Texel Size}, B=Multiply(A=@n5) {constB=0.5}) {constB=2}
  n8 = MaterialFunctionCall(Float2=@n1) {materialFunction=&BreakOutFloat2Components}
  n9 = Divide(
   A=MaterialFunctionCall!(2D Index=Floor(None=@n2), Columns X=@n8.R) {materialFunction=/Engine/Functions/Engine_MaterialFunctions01/Coordinates/2Dto1DIndex}.1D Index
   B=Multiply(A=@n8.R, B=@n8.R)
  )
  FunctionOutput(None=AppendVector(A=Subtract(A=@n6, B=@n7), B=Add(A=@n9) {constB=0, desc=Precision Fix for converting back to 2d later. 0.00001})) {description=The 3D Coordinates laid out in 2D.}
 ]
}
