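# Unlit debug material in the fluid simulation tools that holds test graphs for volume-texture coordinates, with no branch shown wired to a material output.
# One branch height-blends two engine textures, using a volume texture sample (texture unset) as transition phase.
# Another builds an edge mask from 8x8 flipbook-encoded volume coordinates.
# A third feeds padded, tiling volume coordinates into inverted, tiling gradient noise.
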Material /Volumetrics/Tools/FluidSimulation/Materials/Debug/Coords
graph {
 properties {pixelDepthOffsetMode=PDOM_Legacy, shadingModel=MSM_Unlit, usedShadingModels=Unlit}
 nodes [
  n1 = TextureSample() {texture=/Engine/EngineMaterials/DefaultDiffuse}
  n2 = MaterialFunctionCall!(
   Texture Object=TextureObject() {texture=null}
   Position
   Scale=ScalarParameter() {defaultValue=500, parameterName=S}
   Num Frames=Constant() {r=16}
   Clamp Z
  ) {materialFunction=/Volumetrics/Content/VolumeTextures/MaterialFunctions/VolumeTextureFunction}
  n3 = Multiply(
   A=MaterialFunctionCall!(Alpha=ComponentMask(None=@n2.Result RGB) {r=true}, Offset=ScalarParameter() {defaultValue=0.25, parameterName=AO}) {materialFunction=/Engine/Functions/Engine_MaterialFunctions01/Opacity/AlphaOffset}.Result
   B=ScalarParameter() {defaultValue=0.7, parameterName=m}
  )
  n4 = OneMinus(None=@n1.G)
  n5 = ScalarParameter() {defaultValue=2, parameterName=H}
  LinearInterpolate(
   A=TextureSample() {texture=/Engine/EditorResources/TilePatine_D}.RGB
   B=@n1.RGB
   Alpha=MaterialFunctionCall(Transition Phase=Clamp(None=@n3), Height Texture=Multiply(A=@n4, B=@n5), Contrast=ScalarParameter() {defaultValue=2, parameterName=C}) {materialFunction=&HeightLerp}.Alpha
  )
  MaterialFunctionCall!(
   UVW=MaterialFunctionCall!(UVs, XY Frames=Constant2Vector() {g=8, r=8}) {materialFunction=/Volumetrics/Content/VolumeTextures/MaterialFunctions/EncodeVolumeCoordinates}.Result
   Falloff Width=ScalarParameter() {defaultValue=0.2, parameterName=R}
   Hardness
   Falloff Power
   XY Frames=Constant() {r=8}
  ) {materialFunction=/Volumetrics/Content/VolumeTextures/MaterialFunctions/VolumeEdgeMask}
  n6 = Constant() {r=4096}
  n7 = ScalarParameter() {defaultValue=4, parameterName=Padding}
  n8 = Constant2Vector() {g=8, r=8}
  n9 = Noise!(
   World Position=Multiply(
    A=MaterialFunctionCall!(Target Resolution=@n6, Edge Pad Texels=@n7, XY Frames=@n8) {materialFunction=/Volumetrics/Content/VolumeTextures/MaterialFunctions/EncodeVolumeCoordinates_Tiling}.Result
    B=ScalarParameter() {defaultValue=4, parameterName=T}
   )
   FilterWidth
  ) {bTiling=true, bTurbulence=false, levels=8, noiseFunction=NOISEFUNCTION_GradientALU, outputMin=0, quality=4, repeatSize=4}
  OneMinus(None=@n9)
 ]
}
