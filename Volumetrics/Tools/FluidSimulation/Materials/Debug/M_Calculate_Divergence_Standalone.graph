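# Unlit debug material that calculates the divergence of a Velocity texture for the fluid simulation tools, with negative emissive values allowed.
# Central differences sample neighbors one texel (1/Resolution) apart, using R along U and G along V.
# Their sum, plus a B-channel difference, is scaled by 0.5/GridScale.
# The B term's V-axis offset lacks the texel-size input, unlike the other two terms.
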Material /Volumetrics/Tools/FluidSimulation/Materials/Debug/M_Calculate_Divergence_Standalone
graph {
 properties {bAllowNegativeEmissiveColor=true, pixelDepthOffsetMode=PDOM_Legacy, shadingModel=MSM_Unlit, usedShadingModels=Unlit}
 nodes [
  n1 = TextureCoordinate()
  n2 = Divide(B=ScalarParameter() {defaultValue=1024, parameterName=Resolution})
  n3 = Multiply(A=@n2, B=Constant2Vector() {r=1})
  n4 = Subtract(
   A=TextureSampleParameter2D(UVs=Add(A=@n1, B=@n3)) {desc=R, parameterName=Velocity, texture=/Volumetrics/Tools/FluidSimulation/Textures/RT_Tiny}.R
   B=TextureSampleParameter2D(UVs=Subtract(A=@n1, B=@n3)) {desc=L, parameterName=Velocity, texture=&RT_Tiny}.R
  )
  n5 = TextureCoordinate()
  n6 = Multiply(A=@n2, B=Constant2Vector() {g=1})
  n7 = Subtract(
   A=TextureSampleParameter2D(UVs=Add(A=@n5, B=@n6)) {desc=T, parameterName=Velocity, texture=&RT_Tiny}.G
   B=TextureSampleParameter2D(UVs=Subtract(A=@n5, B=@n6)) {desc=B, parameterName=Velocity, texture=&RT_Tiny}.G
  )
  n8 = TextureCoordinate()
  n9 = Multiply(B=Constant2Vector() {g=1})
  n10 = TextureSampleParameter2D(UVs=Add(A=@n8, B=@n9)) {desc=T, parameterName=Velocity, texture=&RT_Tiny}
  n11 = TextureSampleParameter2D(UVs=Subtract(A=@n8, B=@n9)) {desc=B, parameterName=Velocity, texture=&RT_Tiny}
  Multiply(
   A=Divide(B=ScalarParameter() {defaultValue=1, parameterName=GridScale}) {constA=0.5, desc=half rdx}
   B=Add(A=Add(A=@n4, B=@n7), B=Subtract(A=@n10.B, B=@n11.B))
  )
 ]
}
