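# Material function that estimates 3D derivatives of a value from four samples at tetrahedrally offset positions, for use with Prepare3DDeriv.
# Sums and differences of the four inputs give the X, Y and Z gradients, each scaled by a quarter over the offset distance.
# The average of the four samples is also output as the estimated value at the original position.
# The optional offset defaults to a tiny distance and must match Prepare3DDeriv.
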MaterialFunction /Engine/Functions/Engine_MaterialFunctions03/Procedurals/Compute3DDeriv
graph {
 properties {bExposeToLibrary=true, description=Uses positions offset in a tetrahedral pattern to compute 3D derivatives. Use with Prepare3DDeriv.}
 nodes [
  n1 = FunctionInput() {description=Value at Offset1 from Prepare3DDeriv, inputName=Value1}
  n2 = FunctionInput() {description=Value at Offset2 from Prepare3DDeriv, inputName=Value2, sortPriority=1}
  n3 = Add(A=@n1, B=@n2)
  n4 = FunctionInput() {description=Value at Offset3 from Prepare3DDeriv, inputName=Value3, sortPriority=2}
  n5 = FunctionInput() {description=Value at Offset4 from Prepare3DDeriv, inputName=Value4, sortPriority=3}
  n6 = Add(A=@n4, B=@n5)
  n7 = FunctionInput() {
   bUsePreviewValueAsDefault=true
   description="Small offset distance to apply from original Position. Default offset (0.0001) should be fine unless using on very small or large scale positions. If changed, plug the same value into Prepare3DDeriv and Compute3DDeriv."
   inputName=Tetrahedral Offset
   inputType=FunctionInput_Scalar
   sortPriority=4
   previewValue=<0.0001 0 0 1>
  }
  n8 = Divide(B=@n7) {constA=0.25}
  FunctionOutput(None=Multiply(A=Subtract(A=@n3, B=@n6), B=@n8) {constB=2500}) {description=Derivative in X direction, outputName=dx}
  n9 = Subtract(A=@n1, B=@n2)
  n10 = Subtract(A=@n4, B=@n5)
  FunctionOutput(None=Multiply(A=Add(A=@n9, B=@n10), B=@n8) {constB=2500}) {description=Derivative in Y direction, outputName=dy, sortPriority=1}
  FunctionOutput(None=Multiply(A=Subtract(A=@n9, B=@n10), B=@n8) {constB=2500}) {description=Derivative in Z direction, outputName=dz, sortPriority=2}
  FunctionOutput(None=Multiply(A=Add(A=@n3, B=@n6)) {constB=0.25}) {description=Estimated value at original position without offsets, outputName=Value, sortPriority=3}
 ]
}
