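# Material function, exposed to the library, that computes the curl of a 3D vector field from the derivatives produced by Prepare3DDeriv/Compute3DDeriv.
# It takes the field's X, Y and Z derivatives as float3 inputs dx, dy and dz and breaks each into components.
# Three subtractions of cross-axis components form the standard curl terms, which MakeFloat3 combines into the Curl output.
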MaterialFunction /Engine/Functions/Engine_MaterialFunctions03/Procedurals/CurlFrom3DDeriv
graph {
 properties {bExposeToLibrary=true, description=Compute curl of at 3D vector field from result of Prepare3DDeriv/Compute3DDeriv}
 nodes [
  n1 = MaterialFunctionCall(Float3=FunctionInput() {description=Derivative in Y direction, inputName=dy, sortPriority=1}) {materialFunction=&BreakOutFloat3Components}
  n2 = MaterialFunctionCall(Float3=FunctionInput() {description=Derivative in Z direction, inputName=dz, sortPriority=2}) {materialFunction=&BreakOutFloat3Components}
  n3 = MaterialFunctionCall(Float3=FunctionInput() {description=Derivative in X direction, inputName=dx}) {materialFunction=&BreakOutFloat3Components}
  FunctionOutput(None=MaterialFunctionCall(X=Subtract(A=@n1.B, B=@n2.G), Y=Subtract(A=@n2.R, B=@n3.B), Z=Subtract(A=@n3.G, B=@n1.R)) {materialFunction=&MakeFloat3}.Result) {description=Curl of vector field from 3D derivatives, outputName=Curl}
 ]
}
