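# Material function Prepare3DDeriv that produces four offset sample positions for computing derivatives of a 3D procedural function.
# It scales four sign vectors forming a tetrahedral pattern by the Tetrahedral Offset input (default 0.0001) and adds each to Position.
# The same 3D function is evaluated at Offset1 through Offset4, and the results feed Compute3DDeriv, which must use the same offset value.
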MaterialFunction /Engine/Functions/Engine_MaterialFunctions03/Procedurals/Prepare3DDeriv
graph {
 properties {
  bExposeToLibrary=true
  description="Uses positions offset in a tetrahedral pattern to compute 3D derivatives. Evaluate the same 3D function at each offset position produced by this function, then feed the resulting values into Compute3DDeriv."
 }
 nodes [
  n1 = FunctionInput() {
   bUsePreviewValueAsDefault=true
   description="Small offset distance to apply from original Position. Default offset (0.0001) should be fine unless using on very small or large scale positions. If changed, plug the same value into Prepare3DDeriv and Compute3DDeriv."
   inputName=Tetrahedral Offset
   inputType=FunctionInput_Scalar
   sortPriority=1
   previewValue=<0.0001 0 0 1>
  }
  n2 = Constant3Vector() {constant=<rgba 1 1 1 0>}
  n3 = FunctionInput() {description=Position to compute derivative, inputName=Position}
  FunctionOutput(None=Add(A=Multiply(A=@n1, B=@n2), B=@n3)) {description=First offset position for computing 3D derivatives., outputName=Offset1}
  n4 = Constant3Vector() {constant=<rgba 1 -1 -1 0>}
  FunctionOutput(None=Add(A=Multiply(A=@n1, B=@n4), B=@n3)) {description=Second offset position for computing 3D derivatives., outputName=Offset2, sortPriority=1}
  n5 = Constant3Vector() {constant=<rgba -1 1 -1 0>}
  FunctionOutput(None=Add(A=Multiply(A=@n1, B=@n5), B=@n3)) {description=Third offset position for computing 3D derivatives., outputName=Offset3, sortPriority=2}
  n6 = Constant3Vector() {constant=<rgba -1 -1 1 0>}
  FunctionOutput(None=Add(A=Multiply(A=@n1, B=@n6), B=@n3)) {description=Fourth offset position for computing 3D derivatives., outputName=Offset4, sortPriority=3}
 ]
}
