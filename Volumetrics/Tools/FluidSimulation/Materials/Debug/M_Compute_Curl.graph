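# Unlit debug material in the fluid simulation tools, a scratchpad comparing ways to compute curl from procedural noise.
# Separate branches try finite-difference gradient noise, built-in curl VectorNoise, and NormalFromFunction crossed with Z.
# Another feeds tetrahedral Voronoi derivatives into CurlFrom3DDeriv, with dx on all three inputs, and subtracts a radial UV tangent.
# No branch is shown connected to a material output.
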Material /Volumetrics/Tools/FluidSimulation/Materials/Debug/M_Compute_Curl
graph {
 properties {bAllowNegativeEmissiveColor=true, pixelDepthOffsetMode=PDOM_Legacy, shadingModel=MSM_Unlit, usedShadingModels=Unlit}
 nodes [
  AppendVector(A=TextureCoordinate(), B=Constant() {r=0.5})
  n1 = WorldPosition()
  n2 = ScalarParameter() {defaultValue=9, parameterName=WP}
  n3 = AppendVector(A=@n5.UV2, B=Constant())
  n4 = AppendVector(A=@n5.UV3, B=Constant())
  n5 = MaterialFunctionCall!(
   Coordinates=ComponentMask(None=Divide(A=@n1.XYZ, B=@n2)) {g=true, r=true}
   Height Map UV Offset
   Normal Map Intensity
   "Function(UV1)"=ComponentMask(None=@n6) {a=true}
   "Function(UV2)"=ComponentMask(None=VectorNoise(World Position=@n3) {noiseFunction=VNF_GradientALU}) {a=true}
   "Function(UV3)"=ComponentMask(None=VectorNoise(World Position=@n4) {noiseFunction=VNF_GradientALU}) {a=true}
  ) {materialFunction=/Engine/Functions/Engine_MaterialFunctions03/Procedurals/NormalFromFunction}
  n6 = VectorNoise(World Position=AppendVector(A=@n5.UV1, B=Constant())) {noiseFunction=VNF_GradientALU}
  ComponentMask(None=@n6) {g=true, r=true}
  Constant()
  CrossProduct(A=@n5.Normal, B=Constant3Vector() {constant=<rgba 0 0 1 0>})
  MaterialFunctionCall!(OffsetAmount=ScalarParameter() {defaultValue=0.002, parameterName=O}) {materialFunction=/Engine/Functions/Engine_MaterialFunctions01/ImageAdjustment/BlurSampleOffsets}
  n7 = AppendVector(A=Multiply(A=TextureCoordinate(), B=ScalarParameter() {defaultValue=9, parameterName=WP}), B=Constant())
  n8 = Add(
   A=@n7
   B=Multiply(A=ScalarParameter() {defaultValue=0.002, parameterName=Offset}, B=Constant3Vector() {constant=<rgba -1 0 0 0>}) {constB=-1}
  )
  n9 = Add(A=@n7, B=Multiply(A=ScalarParameter() {defaultValue=0.002, parameterName=Offset}, B=Constant3Vector() {constant=<rgba 1 0 0 0>}) {constB=-1})
  Multiply(
   A=Subtract(
    A=Noise!(World Position=@n8, FilterWidth) {bTurbulence=false, desc=L, levels=1, noiseFunction=NOISEFUNCTION_GradientALU, outputMin=0}
    B=Noise(World Position=@n9) {bTurbulence=false, desc=R, levels=1, noiseFunction=NOISEFUNCTION_GradientALU, outputMin=0}
   )
   B=ScalarParameter() {defaultValue=50, parameterName=m}
  )
  Multiply(A=ComponentMask(None=VectorNoise(World Position=@n7) {bTiling=true, noiseFunction=VNF_CurlALU, tileSize=27}) {g=true, r=true}) {constB=0.25}
  Noise(World Position=@n7) {bTurbulence=false, desc=Center Sample, levels=1, noiseFunction=NOISEFUNCTION_GradientALU, outputMin=0}
  n10 = Subtract(A=TextureCoordinate()) {constB=0.5}
  Normalize(VectorInput=@n10)
  n11 = Constant()
  Normalize(VectorInput=CrossProduct(A=AppendVector(A=@n10, B=@n11), B=Constant3Vector() {constant=<rgba 0 0 1 0>}))
  n12 = ScalarParameter() {defaultValue=0.001, parameterName=TA}
  n13 = MaterialFunctionCall!(Position=Divide(A=WorldPosition().XYZ, B=ScalarParameter() {defaultValue=9, parameterName=WP}), Tetrahedral Offset=@n12) {materialFunction=/Engine/Functions/Engine_MaterialFunctions03/Procedurals/Prepare3DDeriv}
  n14 = MaterialFunctionCall!(
   Value1=Noise(World Position=@n13.Offset1) {bTurbulence=false, levels=2, noiseFunction=NOISEFUNCTION_VoronoiALU, quality=3}
   Value2=Noise(World Position=@n13.Offset2) {bTurbulence=false, levels=2, noiseFunction=NOISEFUNCTION_VoronoiALU, quality=3}
   Value3=Noise(World Position=@n13.Offset3) {bTurbulence=false, levels=2, noiseFunction=NOISEFUNCTION_VoronoiALU, quality=3}
   Value4=Noise(World Position=@n13.Offset4) {bTurbulence=false, levels=2, noiseFunction=NOISEFUNCTION_VoronoiALU, quality=3}
   Tetrahedral Offset=@n12
  ) {materialFunction=/Engine/Functions/Engine_MaterialFunctions03/Procedurals/Compute3DDeriv}
  n15 = TextureCoordinate()
  n16 = Constant() {r=0.5}
  n17 = AppendVector(A=Normalize(VectorInput=Subtract(A=@n15, B=@n16)), B=Constant())
  n18 = Constant3Vector() {constant=<rgba 0 0 1 0>}
  Subtract(
   A=ComponentMask(None=MaterialFunctionCall!(dx=@n14.dx, dy=@n14.dx, dz=@n14.dx) {materialFunction=/Engine/Functions/Engine_MaterialFunctions03/Procedurals/CurlFrom3DDeriv}.Curl) {b=true, g=true, r=true}
   B=ComponentMask(None=CrossProduct(A=@n17, B=@n18)) {b=true, g=true, r=true}
  )
  n19 = Multiply(A=ScalarParameter() {defaultValue=0.002, parameterName=Offset}, B=Constant3Vector() {constant=<rgba 0 -1 0 0>}) {constB=-1}
  n20 = Multiply(A=ScalarParameter() {defaultValue=0.002, parameterName=Offset}, B=Constant3Vector() {constant=<rgba 0 1 0 0>}) {constB=-1}
  Subtract(
   A=Noise(World Position=Add(A=@n7, B=@n19)) {bTurbulence=false, desc=B, levels=1, noiseFunction=NOISEFUNCTION_GradientALU, outputMin=0}
   B=Noise(World Position=Add(A=@n7, B=@n20)) {bTurbulence=false, desc=T, levels=1, noiseFunction=NOISEFUNCTION_GradientALU, outputMin=0}
  )
  TextureSample(UVs=TextureCoordinate())
 ]
}
