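# Unlit debug material for the volumetric fluid simulation tools that generates a synthetic divergence test pattern.
# EncodeVolumeCoordinates maps UVs with a 16-frame XY layout, feeding CheckerPattern with a parameterized tiling.
# The checker blends between equal positive and negative values in the blue channel.
# Negative emissive color is allowed, so signed values can be written out.
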Material /Volumetrics/Tools/FluidSimulation/Materials/Debug/M_Create_Divergence
graph {
 properties {bAllowNegativeEmissiveColor=true, pixelDepthOffsetMode=PDOM_Legacy, shadingModel=MSM_Unlit, usedShadingModels=Unlit}
 nodes [
  n1 = Constant() {r=16}
  LinearInterpolate(
   A=Constant3Vector() {constant=<rgba 0 0 10 0>}
   B=Constant3Vector() {constant=<rgba 0 0 -10 0>}
   Alpha=MaterialFunctionCall!(
    UVs=MaterialFunctionCall!(UVs, XY Frames=@n1) {materialFunction=/Volumetrics/Content/VolumeTextures/MaterialFunctions/EncodeVolumeCoordinates}.Result
    Tiling=ScalarParameter() {defaultValue=2, parameterName=456}
   ) {materialFunction=/Engine/ArtTools/RenderToTexture/MaterialFunctions/CheckerPattern}.Result
  )
  TextureCoordinate()
  TextureCoordinate()
 ]
}
