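# Debug test material in the fluid simulation tools for reading a volume stored as a 2D tiled texture.
# EncodeVolumeCoordinates converts UVs into volume positions using XYFrames, and a vector parameter offsets them.
# VolumeTextureFunction then samples the Source texture, described as Velocity, at those positions with the same frame count.
# No material output is wired, and a second Source sample node sits unconnected.
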Material /Volumetrics/Tools/FluidSimulation/Materials/Debug/M_Encode_3d_Test
graph {
 properties {pixelDepthOffsetMode=PDOM_Legacy, usedShadingModels=Default Lit}
 nodes [
  n1 = TextureObjectParameter() {
   desc=Velocity
   parameterName=Source
   samplerType=SAMPLERTYPE_LinearColor
   texture=/Engine/Functions/Engine_MaterialFunctions02/PivotPainter2/Black_1x1_EXR_Texture
  }
  n2 = ScalarParameter() {defaultValue=10, parameterName=XYFrames}
  MaterialFunctionCall!(
   Texture Object=@n1
   Position=Add(
    A=MaterialFunctionCall!(UVs, XY Frames=@n2) {materialFunction=/Volumetrics/Content/VolumeTextures/MaterialFunctions/EncodeVolumeCoordinates}.Result
    B=VectorParameter().RGB
   )
   Scale=Constant() {r=1}
   Num Frames=@n2
   Clamp Z
  ) {materialFunction=/Volumetrics/Content/VolumeTextures/MaterialFunctions/VolumeTextureFunction}
  TextureSampleParameter2D() {desc=Velocity, parameterName=Source, texture=null}
 ]
}
