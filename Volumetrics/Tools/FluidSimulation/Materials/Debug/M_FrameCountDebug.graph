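# Unlit debug material for the FluidSimulation tools that visualizes the XY frame tiling of a flipbook-style volume texture layout.
# UVs scaled by the XYFrames parameter are floored into a tile index, which DebugFloat2Values prints color-coded inside each center-scaled tile.
# Floor and saturate masks on the tile-local UV, assuming 256 steps per tile, add edge lines over the digits.
# An EncodeVolumeCoordinates call and a Subtract node are left unused.
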Material /Volumetrics/Tools/FluidSimulation/Materials/Debug/M_FrameCountDebug
graph {
 properties {bAllowNegativeEmissiveColor=true, pixelDepthOffsetMode=PDOM_Legacy, shadingModel=MSM_Unlit, usedShadingModels=Unlit}
 nodes [
  n1 = TextureCoordinate()
  n2 = ComponentMask(None=VectorParameter() {parameterName=XYFrames, defaultValue=<rgba 16 16 0 0>}.RGB) {g=true, r=true}
  n3 = Multiply(A=@n1, B=@n2)
  n4 = Frac(None=@n3)
  n5 = ScalarParameter() {defaultValue=2}
  n6 = Floor(None=Multiply(A=@n4) {constB=256})
  n7 = MaterialFunctionCall(Float2=Saturate(None=@n6)) {materialFunction=&BreakOutFloat2Components}
  n8 = Min(A=@n7.R, B=@n7.R)
  n9 = MaterialFunctionCall(Float2=Saturate(None=Subtract(A=@n6) {constB=254})) {materialFunction=&BreakOutFloat2Components}
  Add(
   A=MaterialFunctionCall!(
    Vector2=Floor(None=@n3)
    MaximumNumberOfDigits
    UVs=MaterialFunctionCall(UVs=@n4, Texture Scale=@n5) {materialFunction=&ScaleUVsByCenter}.UVs
    DebugTextLocation RG_UpperRight BA_LowerLeft
    Component Spacing
   ) {materialFunction=/Engine/Functions/Engine_MaterialFunctions02/Utility/DebugFloat2Values}.ColorCodedOutput
   B=Max(A=OneMinus(None=@n8), B=Max(A=@n9.R, B=@n9.R))
  )
  MaterialFunctionCall!(UVs=@n1, XY Frames=@n2) {materialFunction=/Volumetrics/Content/VolumeTextures/MaterialFunctions/EncodeVolumeCoordinates}
  Subtract(A=@n4)
 ]
}
