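# Material function in the Coordinates library category that converts a 1D index into a 3D cell index for a grid sized by Cells XYZ.
# It rounds up the index and the X and Y sizes, then derives each axis with Fmod, division and Floor in X-fastest, Z-slowest order.
# A second output adds a half-cell offset and divides by the cell counts, giving the cell's centered UV position.
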MaterialFunction /Engine/Functions/Engine_MaterialFunctions01/Coordinates/1Dto3DIndex
graph {
 properties {
  bExposeToLibrary=true
  description="Converts a 2D index into a 1D index based on a given total X,Y size. For example an 8x8 grid would return cells 0-63."
  libraryCategoriesText [Coordinates]
 }
 nodes [
  n1 = Ceil(None=FunctionInput() {bUsePreviewValueAsDefault=true, inputName=1D Index, inputType=FunctionInput_Scalar})
  n2 = FunctionInput() {bUsePreviewValueAsDefault=true, inputName=Cells XYZ, sortPriority=1}
  n3 = MaterialFunctionCall(Float3=@n2) {materialFunction=&BreakOutFloat3Components}
  n4 = Ceil(None=@n3.R)
  n5 = Multiply(A=@n4, B=Ceil(None=@n3.G))
  n6 = Fmod(A=@n1, B=@n5)
  n7 = MaterialFunctionCall(X=Fmod(A=@n1, B=@n4), Y=Floor(None=Divide(A=@n6, B=@n4)), Z=Floor(None=Divide(A=@n1, B=@n5))) {materialFunction=&MakeFloat3}
  FunctionOutput(None=@n7.Result) {outputName=3D Index}
  FunctionOutput(None=Divide(A=Add(A=@n7.Result) {constB=0.5}, B=Multiply(A=@n2))) {outputName=Centered UV Position, sortPriority=1}
 ]
}
