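# Unlit, two-sided debug material graph for previewing fluid simulation motion vectors as lines in a 3D voxel grid.
# Mesh UVs plus a start offset give each polygon a texel index, mapped to a voxel position and a flipbook-atlas sample of the vector texture.
# Offset math collapses vertices onto that voxel, then stretches them along the vector and widens them perpendicular to the camera ray.
# An alternative fixed-size quad offset and debug-value nodes sit alongside.
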Material /Volumetrics/Tools/FluidSimulation/Materials/Debug/M_PreviewMotionVector_Grid_01
graph {
 properties {pixelDepthOffsetMode=PDOM_Legacy, shadingModel=MSM_Unlit, twoSided=true, usedShadingModels=Unlit}
 nodes [
  n1 = WorldPosition()
  n2 = TextureCoordinate()
  n3 = ScalarParameter() {defaultValue=512, parameterName=PolysPerSize-Mesh}
  n4 = Add(
   A=MaterialFunctionCall!(2D Index=Multiply(A=@n2, B=@n3), Columns X=@n3) {desc=Current 1d Index, materialFunction=/Engine/Functions/Engine_MaterialFunctions01/Coordinates/2Dto1DIndex}.1D Index
   B=ScalarParameter() {defaultValue=50000, parameterName=StartTexel}
  )
  n5 = ScalarParameter() {defaultValue=96, parameterName=TexelsPerSide-Texture}
  n6 = MaterialFunctionCall!(1D Index=@n4, Cells XYZ=@n5) {materialFunction=/Engine/Functions/Engine_MaterialFunctions01/Coordinates/1Dto3DIndex}
  n7 = Multiply(
   A=ConstantBiasScale(None=Divide(A=@n6.3D Index, B=@n5) {desc=0 to 1}) {bias=-0.5, desc=-1 to 1, scale=2}
   B=ScalarParameter() {defaultValue=128, parameterName=VolumeSize}
  )
  n8 = Add(A=Add(A=Multiply(A=@n1.XYZ) {constB=-1, desc="Neutralize Verts to 0,0,0"}, B=ActorPositionWS()) {desc=Center on Object}, B=@n7)
  n9 = Divide(
   A=Subtract(A=WorldPosition() {worldPositionShaderOffset=WPT_ExcludeAllShaderOffsets}.XYZ, B=ObjectPositionWS())
   B=ScalarParameter() {defaultValue=100, parameterName=Encoded Poly Size}
  ) {constB=20, desc=-0.5 to 0.5}
  n10 = Divide(A=Add(A=ComponentMask(None=@n9) {g=true, r=true}, B=Constant2Vector() {r=1}) {constB=0.5}, B=Constant2Vector() {g=1, r=2})
  n11 = ConstantBiasScale(None=@n10) {bias=-0.5, scale=2}
  n12 = Constant()
  Add(A=@n8, B=Multiply(A=AppendVector(A=@n11, B=@n12), B=ScalarParameter() {defaultValue=10, parameterName=Poly Size}))
  n13 = ScalarParameter() {defaultValue=96, parameterName=TexelsPerSide-Texture}
  n14 = Multiply(A=@n13, B=@n13)
  n15 = Divide(A=@n4, B=@n14)
  n16 = Divide(
   A=MaterialFunctionCall!(1D Index=Floor(None=@n15) {desc=Which Frame}, Cells XY=ScalarParameter() {defaultValue=14, parameterName=Frames per Side}) {desc=Find Frame Coordinate, materialFunction=/Engine/Functions/Engine_MaterialFunctions01/Coordinates/1Dto2DIndex}.2D Index
   B=ScalarParameter() {defaultValue=14, parameterName=Frames per Side}
  ) {desc=Normalize in 0-1 UV space}
  n17 = ScalarParameter() {defaultValue=96, parameterName=TexelsPerSide-Texture}
  n18 = Divide(
   A=MaterialFunctionCall(1D Index=Fmod(A=@n4, B=@n14) {desc=Which Texel within Frame}, Cells XY=@n17) {desc=Texel Location within Frame, materialFunction=&1Dto2DIndex}.2D Index
   B=@n17
  ) {desc=Normalize in 0-1 UV space}
  n19 = ScalarParameter() {defaultValue=14, parameterName=Frames per Side}
  n20 = TextureSampleParameter2D(UVs=Add(A=@n16, B=Divide(A=@n18, B=@n19) {desc=Normalize in 0-1 frame space}) {desc=Actual Sample Location}) {desc=Motion Vectors, parameterName=RT, texture=/Volumetrics/Tools/FluidSimulation/Textures/RT_Tiny}
  n21 = Multiply(A=@n20.RGB, B=ScalarParameter() {defaultValue=10, parameterName=Motion Scale})
  n22 = ComponentMask(None=@n10) {g=true}
  n23 = Normalize(VectorInput=@n20.RGB)
  n24 = ActorPositionWS()
  n25 = Normalize(VectorInput=Subtract(A=Add(A=@n7, B=@n24) {desc=World Position of Voxel}, B=CameraPositionWS())) {desc=Camera to Voxel}
  n26 = Multiply(A=Normalize(VectorInput=CrossProduct(A=@n23, B=@n25)), B=ComponentMask(None=@n10) {r=true})
  n27 = ScalarParameter() {defaultValue=10, parameterName=Line Width}
  Add(A=Add(A=Multiply(A=@n21, B=@n22), B=Multiply(A=@n26, B=@n27)), B=@n8)
  ComponentMask(None=VectorParameter() {parameterName=fasdf, defaultValue=<rgba 128 126 0 0>}.RGB) {g=true, r=true}
  Constant3Vector() {constant=<rgba 0 1 0 0>}
  MaterialFunctionCall() {materialFunction=&DebugScalarValues}
  MaterialFunctionCall!(Vector2=@n16, MaximumNumberOfDigits, UVs, DebugTextLocation RG_UpperRight BA_LowerLeft, Component Spacing) {materialFunction=/Engine/Functions/Engine_MaterialFunctions02/Utility/DebugFloat2Values}
 ]
}
