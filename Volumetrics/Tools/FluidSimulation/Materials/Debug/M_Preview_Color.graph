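# M_Preview_Color is an unlit debug material that previews the Color texture of the fluid simulation tool.
# It raises the Color alpha to an adjustable power and scales it into a temperature for a BlackBody node.
# The resulting glow is added to the clamped Color RGB, then multiplied by the blue channel of the Obstacle texture as a mask.
# Both textures are parameters that default to a placeholder render target.
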Material /Volumetrics/Tools/FluidSimulation/Materials/Debug/M_Preview_Color
graph {
 properties {bAllowNegativeEmissiveColor=true, pixelDepthOffsetMode=PDOM_Legacy, shadingModel=MSM_Unlit, usedShadingModels=Unlit}
 nodes [
  n1 = TextureSampleParameter2D() {parameterName=Color, samplerType=SAMPLERTYPE_LinearColor, texture=/Volumetrics/Content/VolumeTextures/Textures/RT_Default}
  n2 = Multiply(
   A=Power(Base=@n1.A, Exp=ScalarParameter() {defaultValue=1.5, parameterName=TempPower})
   B=ScalarParameter() {defaultValue=12000, parameterName=T}
  )
  n3 = TextureSampleParameter2D() {desc=L, parameterName=Obstacle, samplerType=SAMPLERTYPE_LinearColor, texture=&RT_Default}
  Multiply(A=Add(A=Clamp(None=@n1.RGB), B=BlackBody!(Temp=@n2)), B=@n3.B)
 ]
}
