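# Debug preview material for the Volumetrics fluid simulation tool that displays a velocity texture as unlit color.
# It samples the Velocity texture parameter as linear color, clamps RGB to zero or above, and scales it by the VelocityMultiplier parameter.
# The clamp discards negative velocity components before scaling, so at a positive multiplier only positive directions are visible, even though negative emissive color is allowed.
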Material /Volumetrics/Tools/FluidSimulation/Materials/Debug/M_Preview_Velocity
graph {
 properties {bAllowNegativeEmissiveColor=true, pixelDepthOffsetMode=PDOM_Legacy, shadingModel=MSM_Unlit, usedShadingModels=Unlit}
 nodes [Multiply(
   A=Max(A=TextureSampleParameter2D() {parameterName=Velocity, samplerType=SAMPLERTYPE_LinearColor, texture=/Volumetrics/Content/VolumeTextures/Textures/RT_Default}.RGB) {constB=0}
   B=ScalarParameter() {defaultValue=1, parameterName=VelocityMultiplier}
  )]
}
