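# Unlit debug material for the fluid simulation tools that computes a 3D radial vector field around the centre of a volume texture.
# EncodeVolumeCoordinates turns UVs and the XYFrames grid into volume coordinates, which are offset to the centre and normalized into outward directions.
# The Attraction parameter scales these unit vectors, so magnitude does not vary with distance.
# Negative emissive colour is allowed, which permits signed vector values.
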Material /Volumetrics/Tools/FluidSimulation/Materials/Debug/M_RadiaLGradient_3D
graph {
 properties {bAllowNegativeEmissiveColor=true, pixelDepthOffsetMode=PDOM_Legacy, shadingModel=MSM_Unlit, usedShadingModels=Unlit}
 nodes [
  n1 = VectorParameter() {parameterName=XYFrames, defaultValue=<rgba 16 16 0 0>}
  n2 = Subtract(A=MaterialFunctionCall!(UVs=TextureCoordinate(), XY Frames=ComponentMask(None=@n1.RGB) {g=true, r=true}) {materialFunction=/Volumetrics/Content/VolumeTextures/MaterialFunctions/EncodeVolumeCoordinates}.Result) {constA=0.5, constB=0.5}
  Multiply(A=Multiply(A=Normalize(VectorInput=@n2), B=ScalarParameter() {defaultValue=0.2, parameterName=Attraction}))
 ]
}
