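# Unlit debug material in the FluidSimulation tools of the Volumetrics plugin, apparently for testing pre-clamped volume texture UVs.
# Three texture samples each take UVs from their own VolumeCoordClamp function call and read mip 0 of a black 1x1 placeholder texture.
# Two Add nodes sum the RGB of the first two samples with the red channel of the third, with no material output connection shown.
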Material /Volumetrics/Tools/FluidSimulation/Materials/Debug/M_UVPreclamp
graph {
 properties {pixelDepthOffsetMode=PDOM_Legacy, shadingModel=MSM_Unlit, usedShadingModels=Unlit}
 nodes [
  n1 = TextureSample!(
   UVs=MaterialFunctionCall() {materialFunction=/Volumetrics/Content/VolumeTextures/MaterialFunctions/VolumeCoordClamp}.UV
   Tex
   Level
   Apply View MipBias
  ) {
   constMipValue=0
   mipValueMode=TMVM_MipLevel
   samplerType=SAMPLERTYPE_LinearColor
   texture=/Engine/Functions/Engine_MaterialFunctions02/PivotPainter2/Black_1x1_EXR_Texture
  }
  n2 = TextureSample(UVs=MaterialFunctionCall() {materialFunction=&VolumeCoordClamp}.UV) {
   constMipValue=0
   mipValueMode=TMVM_MipLevel
   samplerType=SAMPLERTYPE_LinearColor
   texture=&Black_1x1_EXR_Texture
  }
  n3 = TextureSample(UVs=MaterialFunctionCall() {materialFunction=&VolumeCoordClamp}.UV) {
   constMipValue=0
   mipValueMode=TMVM_MipLevel
   samplerType=SAMPLERTYPE_LinearColor
   texture=&Black_1x1_EXR_Texture
  }
  Add(A=Add(A=@n1.RGB, B=@n2.RGB), B=@n3.R)
 ]
}
